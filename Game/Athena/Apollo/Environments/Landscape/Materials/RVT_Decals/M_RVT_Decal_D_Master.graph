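# Master material for translucent landscape decals that output diffuse color and opacity to a Runtime Virtual Texture, supporting instanced meshes.
# Optional procedural UVs pick atlas slices from fractional pivot or vertex world Z, with a tire-track variant.
# Switches add inverted-color tinting, red vertex-color alpha and height-fraction opacity.
# Attributes feed the RVT output, while a virtual texture feature switch routes shading through EnableRVT or uses them directly.
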Material /Game/Athena/Apollo/Environments/Landscape/Materials/RVT_Decals/M_RVT_Decal_D_Master
graph {
 properties {
  bAutomaticallySetUsageInEditor=false
  bCastRayTracedShadows=false
  bTangentSpaceNormal=false
  bUseMaterialAttributes=true
  bUsedWithInstancedStaticMeshes=true
  blendMode=BLEND_Translucent
  pixelDepthOffsetMode=PDOM_Legacy
  usedShadingModels=Default Lit
 }
 nodes [
  Constant2Vector() {g=1, r=1}
  n1 = Frac(None=ComponentMask(None=WorldPosition().XYZ) {b=true})
  n2 = ScalarParameter() {defaultValue=2, group=Procedural UV, parameterName=U Slices, sliderMax=256, sliderMin=1}
  n3 = TextureCoordinate()
  n4 = Add(A=Multiply(A=Round(None=@n1), B=Divide(B=@n2)) {constB=0.5}, B=@n3)
  n5 = ScalarParameter() {defaultValue=1, group=Procedural UV, parameterName=V Slices, sliderMax=256, sliderMin=1}
  n6 = Divide(
   A=ComponentMask(None=VectorParameter() {group=Procedural UV, parameterName=Base UV Transform, defaultValue=<rgba 1 1 0 0>}.RGB) {g=true, r=true}
   B=MaterialFunctionCall(X=@n2, Y=@n5) {materialFunction=&MakeFloat2}.Result
  )
  n7 = StaticSwitchParameter(
   True=WorldPosition() {worldPositionShaderOffset=WPT_ExcludeAllShaderOffsets}.XYZ
   False=MaterialFunctionCall() {materialFunction=&ObjectPivotPoint}.Object Pivot Location
  ) {
   desc="Vertex Z Fractional Position for UV offsets, otherwise use object pivot Z Fractional Position"
   group=Procedural UV
   parameterName=Vertex Z Pos
  }
  n8 = ComponentMask(None=@n7) {b=true}
  n9 = Multiply(A=Multiply(A=@n2, B=@n5), B=Frac(None=Abs(None=@n8)))
  n10 = Floor(None=Divide(A=@n9, B=@n2) {desc=Y result})
  n11 = Floor(None=Subtract(A=@n9, B=Multiply(A=@n2, B=@n10)) {desc=X result})
  n12 = Add(
   A=Multiply(A=@n6, B=MaterialFunctionCall(X=@n11, Y=@n10) {materialFunction=&MakeFloat2}.Result)
   B=Multiply(A=@n3, B=@n6)
  )
  n13 = TextureSampleParameter2D(UVs=StaticSwitchParameter(True=StaticSwitchParameter(True=@n4, False=@n12) {group=Procedural UV, parameterName=Tire Tracks Tile in V}, False=@n3) {defaultValue=true, group=Procedural UV, parameterName=Procedural UV}) {parameterName=Diffuse}
  n14 = OneMinus(None=@n13.RGB)
  n15 = VectorParameter() {group=Diffuse, parameterName=Diffuse Color Tint, defaultValue=<rgba 1 1 1 1>}
  n16 = Frac(None=ComponentMask(None=WorldPosition() {worldPositionShaderOffset=WPT_ExcludeAllShaderOffsets}.XYZ) {b=true})
  n17 = StaticSwitchParameter(True=Multiply(A=VertexColor().R, B=@n13.A), False=@n13.A) {group=Opacity, parameterName=Alpha Blend with Red Vert Color}
  n18 = ScalarParameter() {defaultValue=1, group=Opacity, parameterName=Opacity, sliderMax=2}
  n19 = SetMaterialAttributes!(
   MaterialAttributes
   Base Color=StaticSwitchParameter(True=Multiply(A=@n14, B=@n15.RGB), False=@n13.RGB) {group=Diffuse, parameterName=Color TInt}
   Specular
   Roughness
   Normal
   Opacity=StaticSwitchParameter(True=Multiply(A=@n16, B=@n17), False=Multiply(A=@n17, B=@n18)) {group=Opacity, parameterName=Opacity Frac Height}
  ) {attributeSetTypes [
    #1
    #2
    #3
    #4
    #5
   ]}
  n20 = GetMaterialAttributes(None=@n19) {attributeGetTypes [
    #1
    #2
    #3
    #4
    #5
   ]}
  RuntimeVirtualTextureOutput!(BaseColor=@n20.Base Color, Specular, Roughness, Normal, WorldHeight, Opacity=@n20.Base Color, Mask, Displacement, Mask4)
  n21 = VirtualTextureFeatureSwitch!(
   No=@n19
   Yes=MaterialFunctionCall!(RVT=@n20.MaterialAttributes, ByPass RVT=@n19) {materialFunction=/Game/Athena/Apollo/Maps/Landscape/VT/MaterialFunctions/EnableRVT}.Result
  )
  ShadingPathSwitch(Default=@n21, Deferred=@n21, Forward=@n21, Mobile=@n21)
 ]
}
