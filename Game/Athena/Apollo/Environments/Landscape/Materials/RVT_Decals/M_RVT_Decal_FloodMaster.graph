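# Translucent master material for flood-themed terrain decals on Apollo landscapes, writing to Runtime Virtual Texture output with an EnableRVT bypass.
# World-aligned dirt textures supply surface attributes, height-blended with dark, glossy puddles driven by two world-space noise samples.
# A radial gradient, height-lerped with texture or noise height and modulated by a world-Z sine, shapes opacity.
# Static switches toggle puddles, texture-channel usage and a flat-color BigSimple variant.
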Material /Game/Athena/Apollo/Environments/Landscape/Materials/RVT_Decals/M_RVT_Decal_FloodMaster
graph {
 properties {
  bAutomaticallySetUsageInEditor=false
  bCastRayTracedShadows=false
  bTangentSpaceNormal=false
  bUseMaterialAttributes=true
  bUsedWithInstancedStaticMeshes=true
  blendMode=BLEND_Translucent
  pixelDepthOffsetMode=PDOM_Legacy
  usedShadingModels=Default Lit
 }
 nodes [
  n1 = VectorParameter() {parameterName=DecalColor, defaultValue=<rgba 0.5182 0.3286 0.1161 0>}
  n2 = ScalarParameter() {defaultValue=0.5, parameterName=SimpleR}
  n3 = ScalarParameter() {defaultValue=0.0165, parameterName=SimpleS}
  n4 = MaterialFunctionCall(Radius=ScalarParameter() {defaultValue=0.6, parameterName=MaskSize}, Density=ScalarParameter() {defaultValue=2.5, parameterName=MaskDensity}) {materialFunction=&RadialGradientExponential}
  n5 = VectorParameter() {parameterName=DecalColor, defaultValue=<rgba 0.5182 0.3286 0.1161 0>}
  n6 = SetMaterialAttributes!(
   MaterialAttributes
   Base Color=Multiply(A=@n5.RGB, B=Constant() {r=0.1}) {constB=0.25}
   Specular=Constant() {r=0.1}
   Roughness=Constant() {r=0.05}
   Normal=Transform(None=Constant3Vector() {constant=<rgba 0 0 1 0>})
  ) {attributeSetTypes [
    #1
    #2
    #3
    #4
   ]}
  n7 = ComponentMask(None=WorldPosition().XYZ) {g=true, r=true}
  n8 = ScalarParameter() {defaultValue=1500, parameterName=Puddles Scale}
  n9 = TextureSample(UVs=Divide(A=@n7, B=@n8)) {samplerSource=SSM_Wrap_WorldGroupSettings, samplerType=SAMPLERTYPE_Grayscale, texture=&LowResBlurredNoise}
  n10 = TextureSample(UVs=Divide(A=@n7, B=Multiply(A=@n8) {constB=1.25})) {samplerSource=SSM_Wrap_WorldGroupSettings, samplerType=SAMPLERTYPE_Grayscale, texture=&LowResBlurredNoise}
  n11 = Min(A=@n9.B, B=@n10.R)
  n12 = TextureObjectParameter() {parameterName=Diffuse, texture=/Game/Environments/Apollo/Sets/Terrain/Textures/T_Apollo_Dirt2_D}
  n13 = TextureObjectParameter() {parameterName=Specular, samplerType=SAMPLERTYPE_Masks, texture=/Game/Environments/Apollo/Sets/Terrain/Textures/T_Apollo_Dirt_S_02}
  n14 = MaterialFunctionCall!(
   Diffuse Texture Object=@n12
   Spec Texture Object=@n13
   TextureSize=ScalarParameter() {defaultValue=300, parameterName=Texture Scale}
   --------------------------
   Use seperate settings for Z projection
   Use seperate texture for Z projection
   Texture Size Z projection
   Z Plane Diffuse Texture Object
   Z Plane Spec Texture Object
   "--------------------------      "
   ProjectionTransitionContrast
   Use Texture Alpha Heightmap to Refine the Transistion
   HeightLerpTransitionContrast
   PickTheDiffuseHeightChannel
   TheDiffuseHeightChannel
   "--------------------------       "
   NormalTextureObject=TextureObjectParameter() {parameterName=Normal, samplerType=SAMPLERTYPE_Normal, texture=/Game/Environments/Apollo/Sets/Terrain/Textures/T_Apollo_Dirt2_N}
   World Space Normals
   "--------------------------  "
   Export Float 4 Diffuse=StaticBool() {value=true}
   Export Float 4 Spec
   "--------------------------     "
   WorldPosition
   World Space Normal
  ) {materialFunction=/Engine/Functions/Engine_MaterialFunctions02/Texturing/WorldAlignedTextures_Complex}
  n15 = MaterialFunctionCall(Float4=@n14.Z Diffuse Texture) {materialFunction=&BreakOutFloat4Components}
  n16 = Constant() {r=5}
  n17 = Constant() {r=3}
  n18 = ComponentMask(None=WorldPosition().XYZ) {b=true}
  n19 = Sine(None=Frac(None=Divide(A=@n18) {constB=1024}))
  n20 = Saturate(None=@n19)
  n21 = MaterialFunctionCall(
   Transition Phase=LinearInterpolate(A=Constant() {r=-0.8}, B=Constant() {r=2}, Alpha=@n11)
   Height Texture=Multiply(A=@n15.A, B=LinearInterpolate(A=@n16, B=@n17, Alpha=@n20) {constA=5})
  ) {materialFunction=&HeightLerp}
  n22 = StaticSwitchParameter(
   True=LinearInterpolate(A=@n5.RGB, B=VectorParameter() {parameterName=Mod Color, defaultValue=<rgba 1 1 1 0>}.RGB, Alpha=@n21.Lerp Alpha No Contrast) {constA=1}
   False=@n5.RGB
  ) {defaultValue=true, parameterName=Pddles True}
  n23 = ComponentMask(None=@n14.Z Diffuse Texture) {b=true, g=true, r=true}
  n24 = MaterialFunctionCall(Float3=@n14.Z Spec) {materialFunction=&BreakOutFloat3Components}
  n25 = ScalarParameter() {defaultValue=1, parameterName=ModRougness}
  n26 = ScalarParameter() {defaultValue=0.2, parameterName=Roughness Min}
  n27 = ScalarParameter() {defaultValue=3, parameterName=Roughness Max}
  n28 = SetMaterialAttributes(
   Base Color=StaticSwitchParameter(True=Multiply(A=@n22, B=@n23), False=Multiply(A=@n15.R, B=@n5.RGB)) {parameterName=Diffuse True}
   Specular=StaticSwitchParameter(True=@n24.R, False=@n15.G) {parameterName=Diffuse True}
   Roughness=StaticSwitchParameter(True=Multiply(A=@n25, B=@n24.B), False=LinearInterpolate(A=@n26, B=@n27, Alpha=@n15.B)) {parameterName=Diffuse True}
   Normal=@n14.Normal Z Texture
  ) {attributeSetTypes [
    #1
    #2
    #3
    #4
   ]}
  n29 = BlendMaterialAttributes(A=@n6, B=@n28, Alpha=Saturate(None=@n21.Alpha))
  n30 = ScalarParameter() {parameterName=BG Alpha Min}
  n31 = ScalarParameter() {defaultValue=1, parameterName=BG Alpha Max}
  n32 = Constant() {r=-1.5}
  n33 = Constant() {r=1.5}
  n34 = WorldPosition()
  n35 = TextureSample(UVs=Divide(A=ComponentMask(None=@n34.XYZ) {g=true, r=true}, B=Constant() {r=1500}), Tex=@n12) {samplerType=SAMPLERTYPE_Grayscale, texture=&LowResBlurredNoise}
  n36 = MaterialFunctionCall(
   Transition Phase=StaticSwitchParameter(True=Multiply(A=@n4.RadialGradientExponential, B=@n19), False=@n4.RadialGradientExponential) {parameterName=Diffuse True}
   Height Texture=StaticSwitchParameter(True=LinearInterpolate(A=@n30, B=@n31, Alpha=@n15.A), False=LinearInterpolate(A=@n32, B=@n33, Alpha=@n35.A)) {parameterName=Diffuse True}
  ) {materialFunction=&HeightLerp}
  n37 = Max(A=@n19) {constB=0.6}
  n38 = Saturate(None=StaticSwitchParameter(True=@n36.Lerp Alpha No Contrast, False=Multiply(A=@n36.Lerp Alpha No Contrast, B=@n37)) {parameterName=Diffuse True})
  n39 = Multiply(A=StaticSwitchParameter(True=Multiply(A=@n38, B=@n11), False=@n38) {defaultValue=true, parameterName=Pddles True}, B=Constant() {r=0.5})
  n40 = StaticSwitchParameter(
   True=SetMaterialAttributes!(
    MaterialAttributes=SetMaterialAttributes!(MaterialAttributes, Base Color=@n1.RGB, Roughness=@n2, Specular=@n3) {attributeSetTypes [#1, #3, #2]}
    Opacity=@n4.RadialGradientExponential
   ) {attributeSetTypes [#5]}
   False=SetMaterialAttributes!(
    MaterialAttributes=StaticSwitchParameter(True=@n29, False=@n28) {defaultValue=true, parameterName=Pddles True}
    Opacity=Subtract(A=@n38, B=@n39)
   ) {attributeSetTypes [#5]}
  ) {parameterName=BigSimple}
  n41 = GetMaterialAttributes(None=@n40) {attributeGetTypes [
    #1
    #2
    #3
    #4
    #5
   ]}
  n42 = ScalarParameter() {defaultValue=1, parameterName=Boost Diffuse}
  RuntimeVirtualTextureOutput!(
   BaseColor=StaticSwitchParameter(True=Multiply(A=@n41.Base Color, B=@n42), False=@n41.Base Color) {parameterName=BigSimple}
   Specular=@n41.Specular
   Roughness=@n41.Specular
   Normal=@n41.Specular
   WorldHeight
   Opacity=@n41.Specular
   Mask
   Displacement
   Mask4
  )
  n43 = VirtualTextureFeatureSwitch!(
   No=@n40
   Yes=MaterialFunctionCall!(RVT=@n41.MaterialAttributes, ByPass RVT=@n40) {materialFunction=/Game/Athena/Apollo/Maps/Landscape/VT/MaterialFunctions/EnableRVT}.Result
  )
  ShadingPathSwitch(Default=@n43, Deferred=@n43, Forward=@n43, Mobile=@n43)
 ]
}
